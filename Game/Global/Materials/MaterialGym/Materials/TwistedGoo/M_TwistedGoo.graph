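# M_TwistedGoo is a Default Lit material in the MaterialGym folder that wraps the MF_TwistedGoo material function, where the goo logic lives.
# It uses material attributes, passing a SetMaterialAttributes node with unwired Base Color, Metallic and Specular pins into the function's input.
# A constant Alpha of one feeds the function, and the graph holds no other shading logic of its own.
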Material /Game/Global/Materials/MaterialGym/Materials/TwistedGoo/M_TwistedGoo
graph {
 properties {bUseMaterialAttributes=true, pixelDepthOffsetMode=PDOM_Legacy, usedShadingModels=Default Lit}
 nodes [MaterialFunctionCall!(
   In=SetMaterialAttributes!(MaterialAttributes, Base Color, Metallic, Specular) {attributeSetTypes [#1, #2, #3]}
   Alpha=Constant() {r=1}
  ) {materialFunction=/Game/Global/Materials/MaterialGym/Materials/TwistedGoo/MaterialFunctions/MF_TwistedGoo}]
}
